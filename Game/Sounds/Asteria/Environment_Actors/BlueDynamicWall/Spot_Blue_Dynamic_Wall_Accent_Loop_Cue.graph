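# Ambient spot Sound Cue for the Blue Dynamic Wall environment actor, layering a continuous loop with intermittent accents.
# A mixer blends a looping wave, identical at every quality level, with a slightly quieter branch that repeatedly plays one of four equally weighted accents after a random delay of several seconds.
# It uses the Athena ambient sound class, a concurrency override limiting simultaneous instances, and a default reverb submix send.
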SoundCue /Game/Sounds/Asteria/Environment_Actors/BlueDynamicWall/Spot_Blue_Dynamic_Wall_Accent_Loop_Cue
root=Mixer_0
nodes [
 {id=Delay_0, children [Random_0], properties {delayMax=8, delayMin=4}}
 {id=Looping_0, children [Delay_0]}
 {id=Mixer_0, children [Looping_0, QualityLevel_0], properties {inputVolume [0.9,1]}}
 {id=QualityLevel_0, children [WavePlayer_1,WavePlayer_1,WavePlayer_1,WavePlayer_1]}
 {id=Random_0, children [WavePlayer_2,WavePlayer_3,WavePlayer_4,WavePlayer_5], properties {weights [1,1,1,1]}}
 {
  id=WavePlayer_1
  properties {bLooping=true, soundWaveAssetPtr=&/Spot_Blue_Dynamic_Wall_Accent_Loop}
 }
 {id=WavePlayer_2, properties {soundWaveAssetPtr=&/Spot_Blue_Dynamic_Wall_Accent_01}}
 {id=WavePlayer_3, properties {soundWaveAssetPtr=&/Spot_Blue_Dynamic_Wall_Accent_02}}
 {id=WavePlayer_4, properties {soundWaveAssetPtr=&/Spot_Blue_Dynamic_Wall_Accent_03}}
 {id=WavePlayer_5, properties {soundWaveAssetPtr=&/Spot_Blue_Dynamic_Wall_Accent_04}}
]
properties {
 bOverrideConcurrency=true
 maxDistance=2097152
 soundClassObject=&Fort_Ambient_Athena
 volumeMultiplier=0.501
 attenuationOverrides {
  bEnableFocusInterpolation=true
  falloffDistance=600
  nonFocusVolumeAttenuation=0.3
  occlusionVolumeAttenuation=0.3
  stereoSpread=0
  attenuationShapeExtents {x=200}
 }
 concurrencyOverrides {maxCount=2}
 soundSubmixSends [{minSendLevel=0.4, soundSubmix=&Fortnite_Reverb_Default}]
}
